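# Unlit, two-sided translucent material for the idle effect on the female Quarrel pickaxe, usable on Niagara mesh and sprite particles.
# Opacity uses panning, noise-distorted airflow minus animated sphere masks, with depth and view-angle fades; a sine wave drives vertex offset, blendable to zero via WPO.
# Emissive blends a curve-atlas gradient into tinted streaks along V, oversaturated by negative desaturation.
# Glider_Glow crossfades color and opacity to a particle-colored radial glow.
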Material /BRCosmetics/Weapons/FORT_Melee/Pickaxe_Quarrel_Female/FX/M_Quarrel_Idle
graph {
 properties {
  bUsedWithNiagaraMeshParticles=true
  bUsedWithNiagaraSprites=true
  blendMode=BLEND_Translucent
  pixelDepthOffsetMode=PDOM_Legacy
  shadingModel=MSM_Unlit
  twoSided=true
  usedShadingModels=Unlit
 }
 nodes [
  n1 = Multiply(A=ComponentMask(None=TextureCoordinate() {coordinateIndex=1}) {g=true})
  n2 = Multiply(A=Time()) {constB=1.5}
  n3 = TextureCoordinate()
  n4 = Multiply(A=Sine(None=Add(A=@n1, B=@n2)), B=OneMinus(None=ComponentMask(None=@n3) {g=true}))
  n5 = MaterialFunctionCall(X=@n4, Y=@n4, Z=@n4) {materialFunction=&MakeFloat3}
  LinearInterpolate(
   A=Multiply(A=Transform(None=@n5.Result) {transformSourceType=TRANSFORMSOURCE_World, transformType=TRANSFORM_Local}) {constB=12}
   B=Constant()
   Alpha=ScalarParameter() {parameterName=WPO}
  )
  n6 = TextureCoordinate() {coordinateIndex=1}
  n7 = VectorParameter() {parameterName=UV, defaultValue=<rgba 1 -0.9 1 1>}
  n8 = AppendVector(A=@n7.R, B=@n7.R)
  n9 = TextureCoordinate() {uTiling=0.5, vTiling=0.5}
  n10 = AppendVector(A=ScalarParameter() {defaultValue=0.1, parameterName=Noise_SpeedU}, B=ScalarParameter() {defaultValue=0.7, parameterName=Noise_SpeedV})
  n11 = Multiply(
   A=TextureSample(UVs=Panner(Coordinate=@n9, Speed=@n10) {speedX=0.1, speedY=0.2}) {samplerType=SAMPLERTYPE_Normal, texture=/Game/Packages/Fortress_Sky/Materials/2Normal}.R
   B=ScalarParameter() {defaultValue=0.4, parameterName=Noise}
  ) {constB=0.18}
  n12 = ComponentMask(None=TextureCoordinate()) {g=true}
  n13 = OneMinus(None=@n12)
  n14 = Panner(
   Coordinate=Add(A=Multiply(A=@n6, B=@n8), B=Multiply(A=@n11, B=@n13))
   Time=Time()
   Speed=AppendVector(A=ScalarParameter() {parameterName=Mask_Speed_U}, B=ScalarParameter() {defaultValue=2, parameterName=Mask_Speed_V})
  ) {speedY=1}
  n15 = Constant()
  n16 = TextureSample(UVs=Panner(Coordinate=TextureCoordinate() {coordinateIndex=1, uTiling=0.2, vTiling=0.4}) {speedY=0.12}) {texture=/Game/Global/Textures/Default/Masks_and_Gradients/T_BlasterBeamDistort}
  n17 = SphereMask(A=@n12, B=@n15) {attenuationRadius=0.7, hardnessPercent=0}
  n18 = Add(A=Multiply(A=SphereMask(A=@n12, B=@n15) {attenuationRadius=1, hardnessPercent=20}, B=Multiply(A=@n16.R)), B=Multiply(A=@n17, B=@n17))
  n19 = DepthFade(
   Opacity=Subtract(A=TextureSample(UVs=@n14) {texture=/Game/Effects/Fort_Effects/Emotes/Skirmish/Skirmish_Male/T_Skirmish_airflow}.R, B=Clamp(None=@n18))
   FadeDistance=ScalarParameter() {defaultValue=20}
  )
  n20 = ParticleColor()
  n21 = ScalarParameter() {defaultValue=0.5, parameterName=Param_3}
  n22 = ParticleColor()
  n23 = MaterialFunctionCall!(Radius, CenterPosition) {materialFunction=/Engine/Functions/Engine_MaterialFunctions01/Gradient/RadialGradient}
  n24 = ScalarParameter() {defaultValue=100, parameterName=NearCameraFadeDistance}
  n25 = Multiply(
   A=DepthFade(
    Opacity=MaterialFunctionCall(Distance=@n24) {materialFunction=&Near_Camera_Fade}.Result
    FadeDistance=DynamicParameter() {parameterIndex=1, defaultValue=<rgba 0 0 0 1>, paramNames [FadeDistance,Param2,Param3,Param4]}.FadeDistance
   )
   B=MaterialFunctionCall("FadeSensitivity (optional)"=ScalarParameter() {defaultValue=0.5, parameterName=AngleFade}) {materialFunction=&FadeBasedOnViewAngle}.Result
  )
  n26 = Constant() {r=1}
  n27 = ScalarParameter() {parameterName=Glider_Glow}
  LinearInterpolate(
   A=Multiply(
    A=Multiply(A=@n19, B=@n20.A)
    B=MaterialFunctionCall("FadeSensitivity (optional)"=@n21) {materialFunction=&FadeBasedOnViewAngle}.Result
   )
   B=Multiply(
    A=Multiply(A=@n22.A, B=@n23.RadialGradient)
    B=ShadingPathSwitch(Default=@n25, Deferred=@n25, Mobile=@n26) {desc=Migrated from FeatureLevelSwitch}
   )
   Alpha=@n27
  )
  n28 = ComponentMask(None=TextureCoordinate()) {g=true}
  n29 = CurveAtlasRowParameter(CurveTime=OneMinus(None=@n28)) {
   atlas=/Game/Athena/LinearColorCurves/VFX_CurveAtlas
   curve=/Game/Weapons/FORT_Melee/Pickaxe_Quarrel_Female/FX/Curve_Quarrel_Female
   defaultValue=26
   parameterName=Curve
  }
  n30 = ComponentMask(None=TextureCoordinate()) {g=true}
  n31 = Constant() {r=1}
  n32 = Panner(Coordinate=Multiply(A=TextureCoordinate(), B=AppendVector(A=@n7.B, B=@n7.B))) {speedY=0.7}
  n33 = Multiply(
   A=Multiply(
    A=SphereMask(A=@n30, B=@n31) {attenuationRadius=0.5, hardnessPercent=0}
    B=TextureSample(UVs=@n32) {texture=&T_Skirmish_airflow}.R
   )
   B=TextureSample(UVs=TextureCoordinate() {vTiling=0.7}) {texture=/Game/CharClasses/Outlander/Gadgets/AntiMaterialCharge/Textures/T_AntimaterialCharge_GroundStreak_02}.R
  )
  n34 = LinearInterpolate(
   A=@n29
   B=Multiply(A=Clamp(None=@n33), B=VectorParameter() {parameterName=LightColor, defaultValue=<rgba 20 17.3001 0 1>}.RGB)
   Alpha=Multiply(A=@n28, B=@n28)
  )
  n35 = Constant() {r=-2}
  MaterialFunctionCall(Emissive=LinearInterpolate(A=Desaturation(None=@n34, Fraction=@n35), B=Multiply(A=@n22.RGB, B=@n23.RadialGradient), Alpha=@n27)) {materialFunction=&MF_FX_EmissiveAdjustments}
 ]
}
