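# Niagara effect for the Quarrel Female pickaxe cosmetic: a brief, quarter-second burst of mesh particles offset from the system origin.
# Empty003 uses three spawn groups to place randomly scaled meshes a third-turn apart, widening as an orange-hot color curve cools and fades.
# Empty004 adds a single mesh that narrows while stretching lengthwise.
# The other four emitters are disabled, including a helix trail and a ribbon sampled from the weapon's Ribbon socket.
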NiagaraSystem /BRCosmetics/Weapons/FORT_Melee/Pickaxe_Quarrel_Female/FX/NS_Pickaxe_Quarrel_Female
dependencies {
 usedDataInterfaces [&NiagaraDataInterfaceCurve, &NiagaraDataInterfaceColorCurve, &NiagaraDataInterfaceArrayFloat3]
 usedDynamicInputs []
 usedModules [&SystemState]
}
emitters {
 Empty {
  bEnabled=false
  emitterUpdateScript [
   &EmitterState(Life Cycle Mode=::NewEnumerator1, Loop Behavior=::NewEnumerator1, Loop Duration=1, UseLoopDelay=false)
   &SpawnBurst_Instantaneous(Use Spawn Probability=false, UseLoopCountLimit=false)
  ]
  particleSpawnScript [
   &InitializeParticle(Lifetime=0.3, Mesh Scale=<1.1 1.25 1.1>, Mesh Scale Mode=::NewEnumerator4, UsePositionOffset=false, Write Material Random=true)
   &CameraOffset(Camera Offset Amount=5)
   &SystemLocation(MaskSpawnGroup=false, Offset=<0 6 47>)
   &InitialMeshOrientation(Rotation=<-0.03 0 0>, Use Orientation Vector=true, Use Rotation Vector=true)
  ]
  particleUpdateScript [
   &ParticleState()
   &UpdateMeshOrientation(Rotation Coordinate Space=::NewEnumerator3, Rotation Vector=<0 1 0>)
   &ScaleColor(
    Scale Alpha=&FloatFromCurve(FloatCurve=DataInterface() {propertyValues {Curve {keys [
         {arriveTangent=5.2267, interpMode=RCIM_Cubic, leaveTangent=5.2267, tangentMode=RCTM_User, time=0, value=0}
         {arriveTangent=1.25, interpMode=RCIM_Cubic, leaveTangent=1.25, time=0.26, value=1.05}
         {arriveTangent=-1.7553, interpMode=RCIM_Cubic, leaveTangent=-1.7553, tangentMode=RCTM_User, time=0.8, value=1}
         {arriveTangent=-6.3132, interpMode=RCIM_Cubic, leaveTangent=-6.3132, tangentMode=RCTM_User, time=1, value=0}
        ]}}})
    ScaleA=true
    ScaleRGB=true
   ) {enabled=false}
   &ScaleMeshSize(Scale Factor=&MakeVector(
     X=&FloatFromCurve(FloatCurve=DataInterface() {propertyValues {Curve {keys [<at 0 1.55 RCIM_Cubic>, {arriveTangent=-0.9135, interpMode=RCIM_Cubic, leaveTangent=-0.9135, tangentMode=RCTM_User, time=1, value=1}]}}})
     Y=&FloatFromCurve(FloatCurve=DataInterface() {propertyValues {Curve {keys [<at 0 1>,<at 1 1.2>]}}})
     Z=&FloatFromCurve(FloatCurve=DataInterface() {propertyValues {Curve {keys [<at 0 1.55 RCIM_Cubic>, {arriveTangent=-0.9135, interpMode=RCIM_Cubic, leaveTangent=-0.9135, tangentMode=RCTM_User, time=1, value=1}]}}})
    ))
  ]
  rendererClasses [&NiagaraMeshRendererProperties]
  renderers [{rendererClass=&NiagaraMeshRendererProperties, rendererIndex=0}]
 }
 Empty001 {
  bEnabled=false
  emitterUpdateScript [
   &EmitterState()
   &FN_SpawnRibbon_v2(SpawnRate=100)
  ]
  particleSpawnScript [
   &InitializeRibbon(
    Color=<rgba 1 0.1624 0 1>
    Lifetime=0.075
    Ribbon Twist Mode:ENiagara_UnsetDirectSet=::NewEnumerator1
    Ribbon Width=50
    UsePositionOffset=false
    Write Material Random=true
   )
   &SkeletalMeshLocation(
    MaskSpawnGroup=false
    Mesh Sampling Type=::NewEnumerator5
    Skeletal Mesh=DataInterface() {propertyValues {
      DefaultMesh=null
      ExcludeBoneName=root
      OptionalComponentClass=null
      PreviewMesh=/BRCosmetics/Weapons/FORT_Melee/Pickaxe_Quarrel_Female/Meshes/Quarrel_Female_Axe
      SkinningMode=SkinOnTheFly
      SoftSourceActor=null
      SourceMode=Default
      UvSetIndex=0
      WholeMeshLOD=-1
      bExcludeBone=false
      bReadDeformedGeometry=true
      bRequireCurrentFrameData=true
      ComponentTags []
      FilteredBones []
      FilteredSockets [Ribbon]
      MeshUserParameter {parameter {name=None, typeDefHandle {registeredTypeIndex=1}, varData []}}
      SamplingRegions []
     }}
    Socket Sampling Mode=::NewEnumerator5
   )
   &AddVelocity(Velocity=&RandomRangeVector(Minimum=<-300 -300 0>, Maximum=<300 300 1>))
  ]
  particleUpdateScript [
   &ParticleState()
   &Drag(Drag=3, Use Linear Drag=true, Use Rotational Drag=false)
   /SetVariables(Particles.Width:NiagaraFloat=&Multiply_Float(
     A=$Particles.RibbonWidth
     B=&FloatFromCurve(FloatCurve=DataInterface() {propertyValues {Curve {keys [{tangentMode=RCTM_User, time=0, value=0}, {tangentMode=RCTM_User, time=1, value=1}]}}})
    )) {isSetParameters=true}
   &SolveForcesAndVelocity(Clamp Velocity=false, Limit Acceleration=false)
   &DynamicMaterialParameters(
    Index 0 Param 4=&FloatFromCurve(FloatCurve=DataInterface() {propertyValues {Curve {keys [<at 0 1>,<at 1 0>]}}})
    Module.Param0WriteEnabled=true
    Module.Param1WriteEnabled=true
    Module.Param2WriteEnabled=true
    Module.Param3WriteEnabled=true
   )
  ]
  rendererClasses [&NiagaraRibbonRendererProperties, &NiagaraSpriteRendererProperties]
  renderers [
   {rendererClass=&NiagaraRibbonRendererProperties}
   {rendererClass=&NiagaraSpriteRendererProperties, rendererIndex=1}
  ]
 }
 Empty002 {
  bEnabled=false
  emitterUpdateScript [
   &EmitterState(Life Cycle Mode=::NewEnumerator1, Loop Behavior=::NewEnumerator1, Loop Duration=0.65, UseLoopDelay=false)
   &SpawnBurst_Instantaneous(Use Spawn Probability=false, UseLoopCountLimit=false)
  ]
  particleSpawnScript [
   &InitializeParticle(Lifetime=0.3, Mesh Scale=<0.9 1 0.9>, Mesh Scale Mode=::NewEnumerator4, UsePositionOffset=false, Write Material Random=true)
   &CameraOffset(Camera Offset Amount=6)
   &SystemLocation(MaskSpawnGroup=false, Offset=<0 50 39>)
   &InitialMeshOrientation(Rotation=<-0.03 0 0>, Use Orientation Vector=true, Use Rotation Vector=true)
  ]
  particleUpdateScript [
   &ParticleState()
   &UpdateMeshOrientation(Rotation Coordinate Space=::NewEnumerator3, Rotation Vector=<0 1 0>)
   &ScaleColor(
    Scale Alpha=&FloatFromCurve(FloatCurve=DataInterface() {propertyValues {Curve {keys [
         {arriveTangent=5.2267, interpMode=RCIM_Cubic, leaveTangent=5.2267, tangentMode=RCTM_User, time=0, value=0}
         {arriveTangent=1.8182, interpMode=RCIM_Cubic, leaveTangent=1.8182, time=0.26, value=1.05}
         {arriveTangent=-1.5611, interpMode=RCIM_Cubic, leaveTangent=-1.5611, tangentMode=RCTM_User, time=0.55, value=1}
         {arriveTangent=-0.8012, interpMode=RCIM_Cubic, leaveTangent=-0.8012, tangentMode=RCTM_User, time=1, value=0}
        ]}}})
    ScaleA=false
    ScaleRGB=true
   ) {enabled=false}
   &ScaleMeshSize(Scale Factor=&MakeVector(
     X=&FloatFromCurve(FloatCurve=DataInterface() {propertyValues {Curve {keys [
          {arriveTangent=0.0034, interpMode=RCIM_Cubic, leaveTangent=0.0034, tangentMode=RCTM_User, time=0, value=1}
          {arriveTangent=0.8206, interpMode=RCIM_Cubic, leaveTangent=0.8206, tangentMode=RCTM_User, time=1, value=1.3}
         ]}}})
     Y=1
     Z=&FloatFromCurve(FloatCurve=DataInterface() {propertyValues {Curve {keys [
          {arriveTangent=0.0033, interpMode=RCIM_Cubic, leaveTangent=0.0033, tangentMode=RCTM_User, time=0, value=1}
          {arriveTangent=0.905, interpMode=RCIM_Cubic, leaveTangent=0.905, tangentMode=RCTM_User, time=1, value=1.3}
         ]}}})
    ))
  ]
  rendererClasses [&NiagaraMeshRendererProperties]
  renderers [{rendererClass=&NiagaraMeshRendererProperties, rendererIndex=0}]
 }
 Empty003 {
  emitterUpdateScript [
   &EmitterState(Life Cycle Mode=::NewEnumerator1, Loop Behavior=::NewEnumerator1, Loop Duration=0.65, UseLoopDelay=false)
   &SpawnBurst_Instantaneous(Use Spawn Probability=false, UseLoopCountLimit=false)
   &SpawnBurst_Instantaneous(Spawn Group=1, Use Spawn Probability=false, UseLoopCountLimit=false) {module=SpawnBurst_Instantaneous001}
   &SpawnBurst_Instantaneous(Spawn Group=2, Use Spawn Probability=false, UseLoopCountLimit=false) {module=SpawnBurst_Instantaneous002}
  ]
  particleSpawnScript [
   &InitializeParticle(
    Color=<rgba 0.5677 0.166 0.0118 1>
    Lifetime=0.24
    Mesh Scale=&RandomRangeVector(Minimum=<0.5 0.7 0.5>, Maximum=<0.8 1.2 0.8>)
    Mesh Scale Mode=::NewEnumerator4
    UsePositionOffset=false
    Write Material Random=true
   )
   &CameraOffset(Camera Offset Amount=6)
   &SystemLocation(MaskSpawnGroup=false, Offset=<0 50 43>)
   &InitialMeshOrientation(Rotation=<-0.02 0 0>, Use Orientation Vector=true, Use Rotation Vector=true)
   &FN_SetSpawnGroup(Number of Spawn Groups=3)
   &InitialMeshOrientation(
    Rotation=&SelectVectorFromArray(
     Direct Array Index=&SpawnGroup()
     Vector Selection Array=DataInterface() {propertyValues {FloatData [<0 0 0>, <0 0.333 0>, <0 0.666 0>]}}
    )
    Use Orientation Vector=true
    Use Rotation Vector=true
   ) {module=InitialMeshOrientation001}
  ]
  particleUpdateScript [
   &ParticleState()
   &UpdateMeshOrientation(Rotation Coordinate Space=::NewEnumerator3, Rotation Rate=0, Rotation Vector=<0 1 0>)
   &ScaleColor(
    Scale Alpha=&FloatFromCurve(FloatCurve=DataInterface() {propertyValues {Curve {keys [
         {arriveTangent=8.1732, time=0, value=0}
         {arriveTangent=1.3125, leaveTangent=1.3125, time=0.35, value=1}
         {arriveTangent=-1.3125, leaveTangent=-1.3125, time=1, value=1}
         {leaveTangent=-6.8984, time=1, value=0}
        ]}}})
    ScaleA=false
    ScaleRGB=true
   )
   &ScaleMeshSize(Scale Factor=&MakeVector(
     X=&FloatFromCurve(FloatCurve=DataInterface() {propertyValues {Curve {keys [
          {arriveTangent=0.0034, interpMode=RCIM_Cubic, leaveTangent=0.0034, tangentMode=RCTM_User, time=0, value=1}
          {arriveTangent=0.8206, interpMode=RCIM_Cubic, leaveTangent=0.8206, tangentMode=RCTM_User, time=1, value=1.3}
         ]}}})
     Y=1
     Z=&FloatFromCurve(FloatCurve=DataInterface() {propertyValues {Curve {keys [
          {arriveTangent=0.0033, interpMode=RCIM_Cubic, leaveTangent=0.0033, tangentMode=RCTM_User, time=0, value=1}
          {arriveTangent=0.905, interpMode=RCIM_Cubic, leaveTangent=0.905, tangentMode=RCTM_User, time=1, value=1.3}
         ]}}})
    ))
   &DynamicMaterialParameters(
    Index 0 Param 3=&RandomRangeFloat(Minimum=0.5, Maximum=1)
    Index 0 Param 4=1
    Index 2 Param 1=10
    Index 2 Param 3=10
    Index 2 Param 4=10
    Module.Param0WriteEnabled=true
    Module.Param0WriteEnabled2=true
    Module.Param1WriteEnabled=true
    Module.Param1WriteEnabled2=true
    Module.Param2WriteEnabled=true
    Module.Param2WriteEnabled2=true
    Module.Param3WriteEnabled=true
    Module.Param3WriteEnabled2=true
    Write Parameter Index 2=true
   )
   /SetVariables(Particles.MeshColor2:LinearColor=&ColorFromCurve(ColorCurve=DataInterface() {propertyValues {
       AlphaCurve {keys [<at 0 1 RCIM_Cubic>,<at 1 0 RCIM_Cubic>]}
       BlueCurve {keys [<at 0 0.5156 RCIM_Cubic>,<at 1 0.2351>]}
       GreenCurve {keys [<at 0 1.5718 RCIM_Cubic>,<at 1 0.4545>]}
       RedCurve {keys [<at 0 11 RCIM_Cubic>,<at 1 2>]}
      }})) {isSetParameters=true}
  ]
  rendererClasses [&NiagaraMeshRendererProperties]
  renderers [
   {rendererClass=&NiagaraMeshRendererProperties, rendererIndex=0}
   {rendererClass=&NiagaraMeshRendererProperties, rendererIndex=1}
  ]
 }
 Empty004 {
  emitterUpdateScript [
   &EmitterState(Life Cycle Mode=::NewEnumerator1, Loop Behavior=::NewEnumerator1, Loop Duration=1, UseLoopDelay=false)
   &SpawnBurst_Instantaneous(Use Spawn Probability=false, UseLoopCountLimit=false)
  ]
  particleSpawnScript [
   &InitializeParticle(Lifetime=0.24, Mesh Scale=<0.35 1 0.35>, Mesh Scale Mode=::NewEnumerator4, UsePositionOffset=false, Write Material Random=true)
   &CameraOffset(Camera Offset Amount=5)
   &SystemLocation(MaskSpawnGroup=false, Offset=<0 48 40>)
   &InitialMeshOrientation(Rotation=<-0.03 0 0>, Use Orientation Vector=true, Use Rotation Vector=true)
  ]
  particleUpdateScript [
   &ParticleState()
   &UpdateMeshOrientation(Rotation Coordinate Space=::NewEnumerator3, Rotation Vector=<0 1 0>)
   &ScaleColor(
    Scale Alpha=&FloatFromCurve(FloatCurve=DataInterface() {propertyValues {Curve {keys [
         {arriveTangent=8.561, time=0, value=0}
         {arriveTangent=1.25, leaveTangent=1.25, time=0.3, value=1}
         {arriveTangent=-1.375, leaveTangent=-1.375, time=1, value=1}
         {leaveTangent=-6.3132, time=1, value=0}
        ]}}})
    ScaleA=true
    ScaleRGB=true
   )
   &ScaleMeshSize(Scale Factor=&MakeVector(
     X=&FloatFromCurve(FloatCurve=DataInterface() {propertyValues {Curve {keys [<at 0 1.55 RCIM_Cubic>, {arriveTangent=-0.9135, interpMode=RCIM_Cubic, leaveTangent=-0.9135, tangentMode=RCTM_User, time=1, value=1}]}}})
     Y=&FloatFromCurve(FloatCurve=DataInterface() {propertyValues {Curve {keys [<at 0 1>,<at 1 1.2>]}}})
     Z=&FloatFromCurve(FloatCurve=DataInterface() {propertyValues {Curve {keys [<at 0 1.55 RCIM_Cubic>, {arriveTangent=-0.9135, interpMode=RCIM_Cubic, leaveTangent=-0.9135, tangentMode=RCTM_User, time=1, value=1}]}}})
    ))
  ]
  rendererClasses [&NiagaraMeshRendererProperties]
  renderers [{rendererClass=&NiagaraMeshRendererProperties, rendererIndex=0}]
 }
 HelixTrail {
  bEnabled=false
  emitterUpdateScript [
   &EmitterState(Life Cycle Mode=::NewEnumerator1, Loop Behavior=::NewEnumerator1, Loop Duration=1, UseLoopDelay=true)
   &SpawnBurst_Instantaneous(Spawn Count=3, Use Spawn Probability=false, UseLoopCountLimit=false)
  ]
  particleSpawnScript [
   &InitializeParticle(
    Color=<rgba 1 1 1 0.5>
    Lifetime Max=0.3
    Lifetime Min=0.28
    Lifetime Mode=::NewEnumerator1
    Mesh Scale Max=<0.4 0.65 0.65>
    Mesh Scale Min=<0.22 0.45 0.45>
    Mesh Scale Mode=::NewEnumerator2
    UsePositionOffset=false
    Write Material Random=true
   )
   &FN_SetSpawnGroup(Number of Spawn Groups=3)
   &InitialMeshOrientation(Rotation=<0 0 0.25>, Use Orientation Vector=true, Use Rotation Vector=true)
   &InitialMeshOrientation(
    Rotation=&SelectVectorFromArray(
     Direct Array Index=$Particles.SpawnGroup
     Vector Selection Array=DataInterface() {propertyValues {FloatData [<0 0 0>, <0 0.333 0>, <0 0.666 0>]}}
    )
    Use Orientation Vector=true
    Use Rotation Vector=true
   ) {module=InitialMeshOrientation002}
   &InitialMeshOrientation(Rotation=<0.15 0 0>, Use Orientation Vector=true, Use Rotation Vector=true) {enabled=false, module=InitialMeshOrientation001}
   /SetVariables(Particles.VelocityDirection:Vector3f=&RotateVectorByQuaternion(Vector To Rotate=<1 0 0>, Quaternion=$Particles.MeshOrientation)) {isSetParameters=true}
   &SystemLocation(MaskSpawnGroup=false, Offset=<0 68 43>)
  ]
  particleUpdateScript [
   &ParticleState()
   &UpdateMeshOrientation(Rotation Rate=1.4, Rotation Vector=$Particles.VelocityDirection)
   &SolveForcesAndVelocity(Clamp Velocity=false, Limit Acceleration=false)
   &SolveRotationalForcesAndVelocity()
   /SetVariables(Particles.PositionOffset:Vector3f=&Add_Vector(
     A=&Add_Vector(A=Linked() {linkedVariable {name=Particles.Position, type=NiagaraPosition}}, B=$Output.SkeletalMeshLocation.SampledPosition)
     B=&Multiply_Vector(
      A=$Particles.VelocityDirection
      B=&VectorFromFloat(Value=&RandomRangeFloat(Minimum=250, Maximum=220))
     )
    )) {enabled=false, isSetParameters=true}
   &ScaleColor(
    Scale Alpha=&FloatFromCurve(
     FloatCurve=DataInterface() {propertyValues {Curve {keys [
         {arriveTangent=6.6269, interpMode=RCIM_Cubic, leaveTangent=6.6269, tangentMode=RCTM_User, time=0, value=0}
         <at 0.1 1 RCIM_Cubic>
         {arriveTangent=-0.3588, interpMode=RCIM_Cubic, leaveTangent=-0.3588, tangentMode=RCTM_User, time=1, value=0}
        ]}}}
     Scale Curve=&RandomRangeFloat(Minimum=0.5, Maximum=1)
    )
    ScaleA=true
    ScaleRGB=true
   )
   &ScaleColor(
    Linear Color Curve=DataInterface() {propertyValues {
      AlphaCurve {keys [<at 0 1>]}
      BlueCurve {keys [<at 0 0.0469>,<at 1 0>]}
      GreenCurve {keys [<at 0 0.3633>,<at 1 0.1055>]}
      RedCurve {keys [<at 0 1>,<at 1 0.5052>]}
     }}
    Scale Mode=::NewEnumerator2
   ) {module=ScaleColor001}
   &ScaleMeshSize(Scale Factor=&MakeVector(
     X=&FloatFromCurve(FloatCurve=DataInterface() {propertyValues {Curve {keys [<at 0 1>,<at 1 1.2>]}}})
     Y=&FloatFromCurve(FloatCurve=DataInterface() {propertyValues {Curve {keys [
          {arriveTangent=5.3995, interpMode=RCIM_Cubic, leaveTangent=5.3995, tangentMode=RCTM_User, time=0, value=0}
          {tangentMode=RCTM_User, time=0.3, value=1}
         ]}}})
     Z=&FloatFromCurve(FloatCurve=DataInterface() {propertyValues {Curve {keys [
          {arriveTangent=5.3995, interpMode=RCIM_Cubic, leaveTangent=5.3995, tangentMode=RCTM_User, time=0, value=0}
          {tangentMode=RCTM_User, time=0.3, value=1}
         ]}}})
    ))
   &DynamicMaterialParameters(
    Index 0 Param 3=&RandomRangeFloat(Minimum=0.5, Maximum=1)
    Index 0 Param 4=&FloatFromCurve(FloatCurve=DataInterface() {propertyValues {Curve {keys [
         {arriveTangent=-1.8203, interpMode=RCIM_Cubic, leaveTangent=-1.8203, tangentMode=RCTM_User, time=0.2, value=1}
         {arriveTangent=-1.1082, interpMode=RCIM_Cubic, leaveTangent=-1.1082, tangentMode=RCTM_User, time=1, value=0}
        ]}}})
    Module.Param0WriteEnabled=false
    Module.Param1WriteEnabled=false
    Module.Param2WriteEnabled=true
    Module.Param3WriteEnabled=true
   )
  ]
  rendererClasses [&NiagaraMeshRendererProperties]
  renderers [{rendererClass=&NiagaraMeshRendererProperties, rendererIndex=0}]
 }
}
summary {
 systemName=NS_Pickaxe_Quarrel_Female
 emitters [
  {bEnabled=false, emitterName=Empty, rendererClasses [&NiagaraMeshRendererProperties]}
  {bEnabled=false, emitterName=Empty002, rendererClasses [&NiagaraMeshRendererProperties]}
  {bEnabled=false, emitterName=HelixTrail, rendererClasses [&NiagaraMeshRendererProperties]}
  {
   bEnabled=false
   emitterName=Empty001
   rendererClasses [&NiagaraRibbonRendererProperties, &NiagaraSpriteRendererProperties]
  }
  {emitterName=Empty003, rendererClasses [&NiagaraMeshRendererProperties]}
  {emitterName=Empty004, rendererClasses [&NiagaraMeshRendererProperties]}
 ]
 userVariables [User.Skeletal Mesh {
   type=NiagaraDataInterfaceSkeletalMesh
   defaultValue {
    struct=NiagaraExt_VariableValue_DataInterface
    value {dataInterface=&self:NiagaraDataInterfaceSkeletalMesh_0}
   }
  }]
}
